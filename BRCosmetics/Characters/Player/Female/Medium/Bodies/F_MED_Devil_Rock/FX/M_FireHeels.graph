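# Translucent, two-sided unlit fire material for the Devil Rock character's heels, usable on skeletal and Niagara meshes.
# Two panned noise samples, smoke and a gradient line form a contrast-sharpened mask that blends two colors into HDR emissive and drives depth-faded opacity, shaped by vertex color.
# World position offset combines noise along normals, a scrolling gradient and height-masked sine waves.
# A per-particle Random offsets the patterns; Distort Edge never applies because its blend alpha self-subtracts.
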Material /BRCosmetics/Characters/Player/Female/Medium/Bodies/F_MED_Devil_Rock/FX/M_FireHeels
graph {
 properties {
  bUsedWithMeshDeformer=true
  bUsedWithMeshParticles=true
  bUsedWithNiagaraMeshParticles=true
  bUsedWithSkeletalMesh=true
  blendMode=BLEND_Translucent
  pixelDepthOffsetMode=PDOM_Legacy
  shadingModel=MSM_Unlit
  twoSided=true
  usedShadingModels=Unlit
 }
 nodes [
  n1 = ComponentMask(None=VectorParameter() {group=UVScale, parameterName=UVScale, defaultValue=<rgba 0.5 0.75 0 0>}.RGB) {g=true, r=true}
  n2 = DynamicParameter() {paramNames [Random,Param2,Param3,Param4]}
  n3 = Add(A=ScalarParameter() {parameterName=UVOffset}, B=@n2.Random)
  n4 = Add(A=Multiply(A=@n1, B=TextureCoordinate() {uTiling=4}), B=@n3)
  n5 = Multiply(A=Time(), B=ScalarParameter() {defaultValue=2, parameterName=Speed})
  n6 = Add(A=@n3, B=Multiply(A=@n1, B=TextureCoordinate() {uTiling=6, vTiling=1.5}))
  n7 = Multiply(
   A=TextureSample(UVs=Panner(Coordinate=@n4, Time=@n5) {speedX=0.1, speedY=1}) {texture=&T_NoisePack_01}.G
   B=TextureSample(UVs=Panner(Coordinate=@n6, Time=@n5) {speedX=-0.1, speedY=1}) {texture=&T_NoisePack_01}.G
  )
  n8 = Multiply(A=VertexNormalWS(), B=@n7)
  n9 = VertexColor()
  n10 = LinearInterpolate(
   A=ScalarParameter() {defaultValue=1.5, parameterName=Distort}
   B=ScalarParameter() {defaultValue=3, parameterName=Distort Edge}
   Alpha=Clamp(None=Subtract(A=@n9.R, B=@n9.R))
  )
  n11 = TextureCoordinate()
  n12 = TextureSample(UVs=Panner(Coordinate=Add(A=@n3, B=@n11), Time=@n5) {speedY=0.5}) {texture=/Game/Global/Textures/Default/Masks_and_Gradients/CylGradient}
  n13 = ComponentMask(None=TextureCoordinate()) {g=true}
  n14 = ScalarParameter() {defaultValue=2, parameterName=WPOHeightMaskPower}
  n15 = Power(Base=OneMinus(None=@n13), Exp=@n14) {constExponent=4}
  n16 = Multiply(A=ScalarParameter() {defaultValue=18, parameterName=WPOHeight}, B=Constant3Vector() {constant=<rgba 0 0 1 0>})
  n17 = Time()
  n18 = ScalarParameter() {defaultValue=0.5, parameterName=WPOWaveSpeed}
  n19 = Add(A=@n13, B=Add(A=Add(A=@n12.G, B=@n2.Random), B=Multiply(A=@n17, B=@n18)))
  n20 = AppendVector(A=Sine(None=@n19) {period=1.1}, B=Sine(None=@n19) {period=0.87})
  n21 = Sine(None=@n19) {period=0.95}
  n22 = Transform(None=Multiply(A=AppendVector(A=@n20, B=@n21), B=VectorParameter() {parameterName=WaveWPO, defaultValue=<rgba 3 1 -1.6 0>}.RGB)) {transformSourceType=TRANSFORMSOURCE_Local}
  Add(A=Add(A=Multiply(A=@n8, B=@n10), B=Multiply(A=@n10, B=@n12.RGB) {constB=35}), B=Multiply(A=@n15, B=Add(A=@n16, B=@n22)))
  Add(A=@n15, B=Power(Base=Multiply(A=@n14) {constB=2}) {constExponent=4})
  n23 = TextureCoordinate() {uTiling=1.5, vTiling=0.6}
  n24 = Add(A=AppendVector(A=ComponentMask(None=@n23) {g=true}, B=ComponentMask(None=@n23) {r=true}), B=AppendVector(A=@n3, B=Constant() {r=-0.27}))
  n25 = Multiply(A=Time(), B=ScalarParameter() {defaultValue=1.5, parameterName=Speed2})
  n26 = Power(Base=TextureSample(UVs=Panner(Coordinate=@n24, Time=@n25) {speedX=0.7}) {samplerType=SAMPLERTYPE_Grayscale, texture=&T_Smoke_Trail_Soft}.G)
  n27 = Add(
   A=Add(A=Multiply(A=@n26) {constB=2}, B=@n7)
   B=Power(
    Base=MaterialFunctionCall() {materialFunction=&LinearGradient}.VGradient
    Exp=ScalarParameter() {defaultValue=15, parameterName=FireLine}
   )
  )
  n28 = ComponentMask(None=TextureCoordinate()) {g=true}
  n29 = Add(A=OneMinus(None=@n28), B=@n9.R)
  n30 = ScalarParameter() {defaultValue=1, parameterName=HardEdgeness}
  n31 = Clamp(None=MaterialFunctionCall(In=Power(Base=@n27) {constExponent=0.5}, Contrast=Multiply(A=@n29, B=@n30)) {materialFunction=&CheapContrast}.Result)
  n32 = ParticleColor()
  n33 = Multiply(
   A=LinearInterpolate(
    A=VectorParameter() {parameterName=Color_A, defaultValue=<rgba 4 0 0.0707 1>}.RGB
    B=VectorParameter() {parameterName=Color_B, defaultValue=<rgba 5 0.6 0 1>}.RGB
    Alpha=@n31
   )
   B=@n32.RGB
  )
  n34 = VectorParameter() {parameterName=HDR_Tint, defaultValue=<rgba 1 1 1 0>}
  MaterialFunctionCall(Emissive=MaterialFunctionCall(LDR Input=Clamp(None=@n33), HDR Tint=Multiply(A=@n33, B=@n34.RGB)) {materialFunction=&DeriveHDRfromLDR}.RemappedRange) {materialFunction=&MF_FX_EmissiveAdjustments}
  n35 = Power(Base=@n28, Exp=ScalarParameter() {defaultValue=4, parameterName=GradientContrast})
  n36 = Multiply(A=@n31, B=Multiply(A=@n35, B=OneMinus(None=@n9.R)))
  n37 = LinearInterpolate(Alpha=@n9.G) {constB=0.5}
  n38 = ScalarParameter() {defaultValue=6, parameterName=FadeDistanceTop}
  n39 = ScalarParameter() {defaultValue=2, parameterName=FadeDistanceBase}
  Multiply(A=@n32.A, B=DepthFade(Opacity=Add(A=@n36, B=@n37), FadeDistance=LinearInterpolate(A=@n38, B=@n39, Alpha=@n35)) {fadeDistanceDefault=50})
 ]
}
