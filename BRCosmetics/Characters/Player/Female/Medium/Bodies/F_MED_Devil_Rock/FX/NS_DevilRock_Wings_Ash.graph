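# Niagara cosmetic effect for the Devil Rock character's wings, emitting sprite ash, flame and wind particles from a skeletal mesh.
# Em_Ash_Wing spawns small embers that rise, stretch with speed and cool from bright orange.
# FlameDecals bursts looping orange sprites that follow the mesh and fade in and out.
# Wind pins cooling wisps to bones or sockets chosen by spawn group, aligning them against a negative-strength point attraction force.
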NiagaraSystem /BRCosmetics/Characters/Player/Female/Medium/Bodies/F_MED_Devil_Rock/FX/NS_DevilRock_Wings_Ash
dependencies {
 usedDataInterfaces [
  &NiagaraDataInterfaceCurve
  &NiagaraDataInterfaceVector2DCurve
  &NiagaraDataInterfaceVectorCurve
  &NiagaraDataInterfaceArrayInt32
  &NiagaraDataInterfaceSkeletalMesh
 ]
 usedDynamicInputs [&InvertBool]
 usedModules [
  &self:SystemSpawnScript.SystemScriptSource.SystemScriptGraph.NiagaraNodeAssignment_0.SetVariables_#1
  &SystemState
 ]
}
emitters {
 Em_Ash_Wing {
  emitterUpdateScript [
   &EmitterState()
   &SpawnRate(SpawnRate=8, Use Spawn Probability=false)
  ]
  particleSpawnScript [
   /SetVariables(Particles.SpriteRotationRate=&RandomRangeFloat(Minimum=-1, Maximum=1)) {isSetParameters=true}
   /SetVariables(Particles.Acceleration:Vector3f=&RandomRangeVector(Minimum=<0 0 10>, Maximum=<0 0 5>)) {isSetParameters=true}
   &InitializeParticle(
    Lifetime=&RandomRangeFloat(Minimum=0.3, Maximum=0.6)
    Sprite Rotation Angle=&RandomRangeFloat(Minimum=0, Maximum=1)
    Sprite Rotation Mode=::NewEnumerator1
    Sprite Size=&Vector2DFromFloat(Value=&RandomRangeFloat(Minimum=1, Maximum=2))
    Sprite Size Mode=::NewEnumerator4
    UsePositionOffset=false
    Write Material Random=true
   )
   &Color(Color=&MakeLinearColorFromVectorAndFloat(
     "Vector (RGB)"=&VectorFromFloat(Value=&RandomRangeFloat(Minimum=0.5, Maximum=1))
     "Float (Alpha)"=&ClampFloat(Float=1)
    ))
   &SkeletalMeshLocation(Bone Sampling Mode=::NewEnumerator0, MaskSpawnGroup=false, OverrideRandomSeed=false, Skeletal Mesh=$System.Skeletal Mesh)
   &ShapeLocation(
    Cylinder Height=20
    Cylinder Radius=10
    MaskSpawnGroup=false
    Offset=<0 20 0>
    Offset Mode=::NewEnumerator0
    Override Seed=false
    Rotation Axis=<0 0 1>
    Rotation Mode=::NewEnumerator1
    Shape Primitive=::NewEnumerator1
    Surface Only=false
    UseOwnerScale=true
   ) {enabled=false}
   &InheritVelocity(Inherited Velocity Amount Scale=<0 0 0>, Inherited Velocity Speed Limit=0, Speed Limit=false, UseSpeedThreshold=false)
   &DynamicMaterialParameters(
    Index 0 Param 1=&RandomRangeFloat(Minimum=0, Maximum=1)
    Index 0 Param 2=&RandomRangeFloat(Minimum=0, Maximum=1)
    Index 0 Param 3=&RandomRangeFloat(Minimum=0.2, Maximum=0.5)
    Index 0 Param 4=&RandomRangeFloat(Minimum=0.5, Maximum=2)
    Module.Index0Param0Write=true
    Module.Index0Param1Write=true
    Module.Index0Param2Write=true
    Module.Index0Param3Write=true
   )
   &AddVelocity(Velocity=&RandomRangeVector(Minimum=<-1 -1 20>, Maximum=<1 1 10>))
  ]
  particleUpdateScript [
   &ParticleState()
   &SpriteRotationRate(Rotation Rate=&AngleConversion(
     Angle:NiagaraFloat=$Particles.SpriteRotationRate
     Angle Input:ENiagara_AngleInput=::NewEnumerator1
     Angle Output:ENiagara_AngleInput=::NewEnumerator0
    ))
   &ScaleSpriteSize(
    Scale Factor=&Vector2DFromCurve(
     Vector2Curve=DataInterface() {propertyValues {
       XCurve {keys [
         <at 0 0 RCIM_Cubic>
         {arriveTangent=1.4286, interpMode=RCIM_Cubic, leaveTangent=1.4286, time=0.12, value=1}
         {arriveTangent=-1.1364, interpMode=RCIM_Cubic, leaveTangent=-1.1364, time=0.7, value=1}
         {interpMode=RCIM_Cubic, leaveTangent=-4.9665, time=1, value=0}
        ]}
       YCurve {keys [
         <at 0 0 RCIM_Cubic>
         {arriveTangent=1.4286, interpMode=RCIM_Cubic, leaveTangent=1.4286, time=0.12, value=1}
         {arriveTangent=-1.1364, interpMode=RCIM_Cubic, leaveTangent=-1.1364, time=0.7, value=1}
         <at 1 0 RCIM_Cubic>
        ]}
      }}
     ScaleCurve=<1 1>
    )
    Scale Sprite Size Mode=::NewEnumerator2
   )
   &ScaleColor(
    Scale Alpha=&FloatFromCurve(FloatCurve=DataInterface() {propertyValues {Curve {keys [<at 0.75 1 RCIM_Cubic>,<at 1 0>]}}})
    Scale RGB=&VectorFromCurve(VectorCurve=DataInterface() {propertyValues {
       XCurve {keys [<at 0 50 RCIM_Cubic>,<at 1 1 RCIM_Cubic>]}
       YCurve {keys [<at 0 15.4515 RCIM_Cubic>,<at 1 0.1222 RCIM_Cubic>]}
       ZCurve {keys [<at 0 2.8538 RCIM_Cubic>,<at 1 0.0824 RCIM_Cubic>]}
      }})
    ScaleA=true
    ScaleRGB=true
   )
   /Niagara/Modules/Update/SubUV/SubUVAnimation(Number Of Frames=0)
   /Niagara/Modules/Update/Forces/PointAttractionForce(Attraction Radius=15, AttractionStrength=-700, AttractorPosition:Vector3f=<1 0 3>, Module.Kill Within Radius=false, Module.Use Falloff=true)
   &AccelerationForce(Acceleration=$Particles.Acceleration)
   &SolveForcesAndVelocity(Clamp Velocity=false, Limit Acceleration=false)
   &ScaleSpriteSizeBySpeed(Max Scale Factor=<10 20>, Source Velocity:Vector3f=&Multiply_Vector(A=<1 0.01 1>, B=$Particles.Velocity))
  ]
  rendererClasses [&NiagaraSpriteRendererProperties]
  renderers [{rendererClass=&NiagaraSpriteRendererProperties}]
 }
 FlameDecals {
  emitterUpdateScript [
   &EmitterState(Inactive Response=::NewEnumerator1, Life Cycle Mode=::NewEnumerator1, Loop Duration=1.25, UseLoopDelay=true)
   &SpawnBurst_Instantaneous(Spawn Count=6, Use Spawn Probability=false, UseLoopCountLimit=false)
  ]
  particleSpawnScript [
   &InitializeParticle(
    Lifetime=&RandomRangeFloat(Minimum=1.5, Maximum=2)
    Sprite Rotation Angle=0.5
    Sprite Rotation Mode=::NewEnumerator1
    Sprite Size=&RandomRangeVector2D(Minimum=<8 20>, Maximum=<16 27>)
    Sprite Size Mode=::NewEnumerator4
    UsePositionOffset=false
    Write Material Random=true
   )
   &Color(Color=&MakeLinearColorFromVectorAndFloat(
     "Vector (RGB)"=<8 1.0499 0>
     "Float (Alpha)"=&ClampFloat(Float=&RandomRangeFloat(Minimum=0.65, Maximum=0.8))
    ))
   &DynamicMaterialParameters(
    Index 0 Param 4=&RandomRangeFloat(Minimum=0, Maximum=1)
    Module.Index0Param0Write=false
    Module.Index0Param1Write=false
    Module.Index0Param2Write=false
    Module.Index0Param3Write=true
   ) {module=DynamicMaterialParameters001}
  ]
  particleUpdateScript [
   &ParticleState()
   &SkeletalMeshLocation(Bone Sampling Mode=::NewEnumerator0, MaskSpawnGroup=false, OverrideRandomSeed=false, Skeletal Mesh=$System.Skeletal Mesh)
   &ScaleColor(
    Scale Alpha=&FloatFromCurve(FloatCurve=DataInterface() {propertyValues {Curve {keys [
         <at 0 0 RCIM_Cubic>
         {arriveTangent=2, interpMode=RCIM_Cubic, leaveTangent=2, time=0.2, value=0.64}
         <at 0.5 1 RCIM_Cubic>
         {arriveTangent=-2, interpMode=RCIM_Cubic, leaveTangent=-2, time=0.8, value=0.64}
         <at 1 0>
        ]}}})
    ScaleA=true
    ScaleRGB=true
   )
   &CameraOffset(Camera Offset Amount=10)
   &DynamicMaterialParameters(
    Index 0 Param 1=1.5
    Index 0 Param 2=1
    Index 0 Param 3=1
    Module.Index0Param0Write=true
    Module.Index0Param1Write=true
    Module.Index0Param2Write=true
    Module.Index0Param3Write=false
   )
   &PointAttractionForce(
    Attraction Radius=700
    AttractionStrength=-700
    AttractorPosition:Vector3f=<12 0 35>
    Module.Kill Within Radius=false
    Module.Use Falloff=true
   )
   &SolveForcesAndVelocity(Clamp Velocity=true, Limit Acceleration=false, Speed Limit=10)
  ]
  rendererClasses [&NiagaraSpriteRendererProperties]
  renderers [{rendererClass=&NiagaraSpriteRendererProperties}]
 }
 Wind {
  emitterUpdateScript [
   &EmitterState()
   &SpawnRate(SpawnRate=8, Use Spawn Probability=false)
  ]
  particleSpawnScript [
   &FN_SetSpawnGroup()
   &InitializeParticle(
    Lifetime=&RandomRangeFloat(Minimum=1.4, Maximum=2)
    Sprite Size=&RandomRangeVector2D(Minimum=<5 40>, Maximum=<12 55>)
    Sprite Size Mode=::NewEnumerator4
    UsePositionOffset=false
    Write Material Random=true
   )
   &Color(Color=&MakeLinearColorFromVectorAndFloat("Vector (RGB)"=<0.2031 0.1309 0.1268>, "Float (Alpha)"=&ClampFloat(Float=1.5)))
   &DynamicMaterialParameters(
    Index 0 Param 1=&RandomRangeFloat(Minimum=0, Maximum=1)
    Index 0 Param 2=&RandomRangeFloat(Minimum=0.52, Maximum=0.7)
    Index 0 Param 3=&RandomRangeFloat(Minimum=0.55, Maximum=0.75)
    Module.Index0Param0Write=true
    Module.Index0Param1Write=true
    Module.Index0Param2Write=true
    Module.Index0Param3Write=false
   )
  ]
  particleUpdateScript [
   &SkeletalMeshLocation(
    Bone / Socket Index=&SelectIntFromArray(Direct Array Index=$Particles.SpawnGroup, Int Selection Array=DataInterface() {propertyValues {IntData [0,2,1,3]}})
    Bone Sampling Mode=::NewEnumerator4
    MaskSpawnGroup=false
    Skeletal Mesh=$System.Skeletal Mesh
   )
   &ParticleState()
   &ScaleSpriteSize(
    Scale Factor=&Vector2DFromCurve(
     Vector2Curve=DataInterface() {propertyValues {
       XCurve {keys [
         <at 0 0.6 RCIM_Cubic>
         {arriveTangent=-0.3, interpMode=RCIM_Cubic, leaveTangent=-0.3, time=0.3766, value=1}
         {interpMode=RCIM_Cubic, leaveTangent=-2.7263, time=1, value=0.3}
        ]}
       YCurve {keys [
         <at 0 0.5 RCIM_Cubic>
         {arriveTangent=0.6, interpMode=RCIM_Cubic, leaveTangent=0.6, time=0.3766, value=0.7054}
         {interpMode=RCIM_Cubic, leaveTangent=0.6503, time=1, value=1.1}
        ]}
      }}
     ScaleCurve=<1 1>
    )
    Scale Sprite Size Mode=::NewEnumerator2
   )
   &ScaleColor(
    Scale Alpha=&FloatFromCurve(FloatCurve=DataInterface() {propertyValues {Curve {keys [<at 0 0 RCIM_Cubic>,<at 0.3 1 RCIM_Cubic>,<at 1 0>]}}})
    Scale RGB=&VectorFromCurve(VectorCurve=DataInterface() {propertyValues {
       XCurve {keys [<at 0 7 RCIM_Cubic>,<at 0.6 0.1588>]}
       YCurve {keys [<at 0 1.3373 RCIM_Cubic>,<at 0.6 0.1684>]}
       ZCurve {keys [<at 0 0.3586 RCIM_Cubic>,<at 0.6 0.1719>]}
      }})
    ScaleA=true
    ScaleRGB=true
   )
   &DynamicMaterialParameters(
    Index 0 Param 4=&FloatFromCurve(FloatCurve=DataInterface() {propertyValues {Curve {keys [<at 0 2 RCIM_Cubic>,<at 0.7 1>]}}})
    Module.Index0Param0Write=false
    Module.Index0Param1Write=false
    Module.Index0Param2Write=false
    Module.Index0Param3Write=true
   ) {module=DynamicMaterialParameters001}
   &PointAttractionForce(
    Attraction Radius=700
    AttractionStrength=-700
    AttractorPosition:Vector3f=<12 0 35>
    Module.Kill Within Radius=false
    Module.Use Falloff=true
   )
   &SolveForcesAndVelocity(Clamp Velocity=true, Limit Acceleration=false, Speed Limit=10)
   &CameraOffset(Camera Offset Amount=10)
   /SetVariables(Particles.AlignmentVector:Vector3f=&Multiply_VectorByFloat(Vector=$Output.SolveForcesAndVelocity.IncomingPhysicsForce, Float=-1)) {isSetParameters=true}
  ]
  rendererClasses [&NiagaraSpriteRendererProperties]
  renderers [{rendererClass=&NiagaraSpriteRendererProperties}]
 }
}
summary {
 systemName=NS_DevilRock_Wings_Ash
 emitters [
  {emitterName=Em_Ash_Wing, rendererClasses [&NiagaraSpriteRendererProperties]}
  {emitterName=Wind, rendererClasses [&NiagaraSpriteRendererProperties]}
  {emitterName=FlameDecals, rendererClasses [&NiagaraSpriteRendererProperties]}
 ]
 userVariables [User.bisFootHidden {type=NiagaraBool}]
}
