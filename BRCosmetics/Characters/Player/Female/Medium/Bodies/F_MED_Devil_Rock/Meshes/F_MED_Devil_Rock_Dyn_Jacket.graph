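# Physics asset for the dynamic jacket on the F_MED_Devil_Rock female player cosmetic, defining capsule bodies and joints for its secondary motion.
# Back, side and front jacket chains hang from spine_04 with limited swing and twist, with two dangles on each front chain's tip.
# The right side chain alone attaches through an intermediate MSTR body.
# Pelvis and spine capsules form the torso, linked by free-rotation constraints.
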PhysicsAsset /BRCosmetics/Characters/Player/Female/Medium/Bodies/F_MED_Devil_Rock/Meshes/F_MED_Devil_Rock_Dyn_Jacket
bodies {
 dangleA_01_l [{shapeType=Capsule, length=6.5, radius=1.25, center=<3.6373 -0.1838 -0.3443>, rotation=<rot -90 -0.0395 0.0394>}]
 dangleA_01_r [{shapeType=Capsule, length=6.5, radius=1.25, center=<-3.5398 -0.1722 -0.3256>, rotation=<rot -90 -0.0395 0.0394>}]
 dangleB_01_l [{shapeType=Capsule, length=5.75, radius=1.25, center=<3.6679 -0.2272 -0.3548>, rotation=<rot 90 0.0395 0.0394>}]
 dangleB_01_r [{shapeType=Capsule, length=5.75, radius=1.25, center=<-3.7115 -0.2121 -0.3292>, rotation=<rot 90 0.0395 0.0394>}]
 dyn_jacket_bk_01 [{shapeType=Capsule, length=3, radius=1, center=<1.5 0 0>, rotation=<rot 90 0 0>}]
 dyn_jacket_bk_02 [{shapeType=Capsule, length=3, radius=1, center=<1.5 0 0>, rotation=<rot 90 0 0>}]
 dyn_jacket_bk_l_01 [{shapeType=Capsule, length=3, radius=1, center=<1.5 0 0>, rotation=<rot 90 0 0>}]
 dyn_jacket_bk_l_02 [{shapeType=Capsule, length=3, radius=1, center=<1.5 0 0>, rotation=<rot 90 0 0>}]
 dyn_jacket_bk_r_01 [{shapeType=Capsule, length=3, radius=1, center=<-1.5 0 0>, rotation=<rot 90 0 0>}]
 dyn_jacket_bk_r_02 [{shapeType=Capsule, length=3, radius=1, center=<-1.5 0 0>, rotation=<rot 90 0 0>}]
 dyn_jacket_fr_l_01 [{shapeType=Capsule, length=3, radius=1, center=<1.5 0 0>, rotation=<rot 90 0 0>}]
 dyn_jacket_fr_l_02 [{shapeType=Capsule, length=3, radius=1, center=<1.8 0 0>, rotation=<rot 90 0 0>}]
 dyn_jacket_fr_l_03 [{shapeType=Capsule, length=0.5, radius=1.5}]
 dyn_jacket_fr_r_01 [{shapeType=Capsule, length=3, radius=1, center=<-1.5 0 0>, rotation=<rot 90 0 0>}]
 dyn_jacket_fr_r_02 [{shapeType=Capsule, length=3, radius=1, center=<-1.8 0 0>, rotation=<rot 90 0 0>}]
 dyn_jacket_fr_r_03 [{shapeType=Capsule, length=0.5, radius=1.5}]
 dyn_jacket_l_01 [{shapeType=Capsule, length=3, radius=1, center=<1.5 0 0>, rotation=<rot 90 0 0>}]
 dyn_jacket_l_02 [{shapeType=Capsule, length=3, radius=1, center=<1.5 0 0>, rotation=<rot 90 0 0>}]
 dyn_jacket_r_01 [{shapeType=Capsule, length=3, radius=1, center=<-1.5 0 0>, rotation=<rot 90 0 0>}]
 dyn_jacket_r_01_MSTR [{shapeType=Capsule, length=0.505, radius=0.505, rotation=<rot 0 -90 0>}]
 dyn_jacket_r_02 [{shapeType=Capsule, length=3, radius=1, center=<-1.5 0 0>, rotation=<rot 90 0 0>}]
 pelvis [{shapeType=Capsule, length=13.25, radius=8.5, center=<0 -0.5054 0>}]
 spine_01 [{shapeType=Capsule, length=3, radius=9.755, rotation=<rot 0 -90 0>}]
 spine_02 [{shapeType=Capsule, length=6.2444, radius=8.475, center=<5.4244 1.7063 0.1138>, rotation=<rot -3.3368 -93.0075 92.5576>}]
 spine_04 [
  {shapeType=Capsule, length=0, radius=9.25, center=<1.37 3.768 0>}
  {shapeType=Capsule, length=1, radius=8, center=<-2.3662 1.4799 -0.1292>, rotation=<rot 0 0 5.0002>}
 ]
 spine_05 [{shapeType=Capsule, length=1, radius=1}]
}
constraints [
 {bone1Name=dyn_jacket_bk_02, bone2Name=dyn_jacket_bk_01, swing1LimitDegrees=20, swing2LimitDegrees=69.8571, twistLimitDegrees=5}
 {bone1Name=dyn_jacket_bk_l_02, bone2Name=dyn_jacket_bk_l_01, swing1LimitDegrees=10, swing2LimitDegrees=5, twistLimitDegrees=5}
 {bone1Name=dyn_jacket_l_02, bone2Name=dyn_jacket_l_01, swing1LimitDegrees=45, swing2LimitDegrees=10, twistLimitDegrees=5}
 {bone1Name=dyn_jacket_fr_l_02, bone2Name=dyn_jacket_fr_l_01, swing1LimitDegrees=45, swing2LimitDegrees=45, twistLimitDegrees=10}
 {bone1Name=dyn_jacket_r_02, bone2Name=dyn_jacket_r_01, swing1LimitDegrees=5, swing2LimitDegrees=0, twistLimitDegrees=5}
 {bone1Name=dyn_jacket_fr_r_02, bone2Name=dyn_jacket_fr_r_01, swing1LimitDegrees=45, swing2LimitDegrees=45, twistLimitDegrees=10}
 {bone1Name=dyn_jacket_bk_r_02, bone2Name=dyn_jacket_bk_r_01, swing1LimitDegrees=10, swing2LimitDegrees=10, twistLimitDegrees=5}
 {bone1Name=dyn_jacket_bk_01, bone2Name=spine_04, swing1LimitDegrees=20, swing2LimitDegrees=57.8571, twistLimitDegrees=5}
 {bone1Name=dyn_jacket_bk_l_01, bone2Name=spine_04, swing1LimitDegrees=10, swing2LimitDegrees=5, twistLimitDegrees=5}
 {bone1Name=dyn_jacket_l_01, bone2Name=spine_04, swing1LimitDegrees=45, swing2LimitDegrees=10, twistLimitDegrees=5}
 {bone1Name=dyn_jacket_fr_l_01, bone2Name=spine_04, swing1LimitDegrees=20, swing2LimitDegrees=20, twistLimitDegrees=10}
 {bone1Name=dyn_jacket_r_01, bone2Name=dyn_jacket_r_01_MSTR, swing1LimitDegrees=5, swing2LimitDegrees=5, twistLimitDegrees=5}
 {bone1Name=dyn_jacket_fr_r_01, bone2Name=spine_04, swing1LimitDegrees=20, swing2LimitDegrees=20, twistLimitDegrees=10}
 {bone1Name=dyn_jacket_bk_r_01, bone2Name=spine_04, swing1LimitDegrees=10, swing2LimitDegrees=5, twistLimitDegrees=5}
 {
  bone1Name=spine_04
  bone2Name=spine_02
  swing1LimitDegrees=45
  swing1Motion=Free
  swing2LimitDegrees=45
  swing2Motion=Free
  twistLimitDegrees=45
  twistMotion=Free
 }
 {
  bone1Name=spine_05
  bone2Name=spine_04
  swing1LimitDegrees=45
  swing1Motion=Free
  swing2LimitDegrees=45
  swing2Motion=Free
  twistLimitDegrees=45
  twistMotion=Free
 }
 {
  bone1Name=dyn_jacket_r_01_MSTR
  bone2Name=spine_04
  swing1LimitDegrees=45
  swing1Motion=Free
  swing2LimitDegrees=45
  swing2Motion=Free
  twistLimitDegrees=45
  twistMotion=Free
 }
 {
  bone1Name=spine_02
  bone2Name=spine_01
  swing1LimitDegrees=45
  swing1Motion=Free
  swing2LimitDegrees=45
  swing2Motion=Free
  twistLimitDegrees=45
  twistMotion=Free
 }
 {bone1Name=dyn_jacket_fr_l_03, bone2Name=dyn_jacket_fr_l_02, swing1LimitDegrees=25, swing2LimitDegrees=25, twistLimitDegrees=10}
 {bone1Name=dyn_jacket_fr_r_03, bone2Name=dyn_jacket_fr_r_02, swing1LimitDegrees=25, swing2LimitDegrees=25, twistLimitDegrees=10}
 {bone1Name=dangleA_01_r, bone2Name=dyn_jacket_fr_r_03, swing1LimitDegrees=45, swing2LimitDegrees=45, twistLimitDegrees=45}
 {bone1Name=dangleB_01_l, bone2Name=dyn_jacket_fr_l_03, swing1LimitDegrees=45, swing2LimitDegrees=45, twistLimitDegrees=45}
 {bone1Name=dangleB_01_r, bone2Name=dyn_jacket_fr_r_03, swing1LimitDegrees=45, swing2LimitDegrees=45, twistLimitDegrees=45}
 {bone1Name=dangleA_01_l, bone2Name=dyn_jacket_fr_l_03, swing1LimitDegrees=45, swing2LimitDegrees=45, twistLimitDegrees=45}
 {
  bone1Name=spine_01
  bone2Name=pelvis
  swing1LimitDegrees=45
  swing1Motion=Free
  swing2LimitDegrees=45
  swing2Motion=Free
  twistLimitDegrees=45
  twistMotion=Free
 }
]
