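# Anim Blueprint for the Devil Rock female medium body cosmetic, adding secondary motion to a pose copied from the attached parent mesh.
# Each update maps spine_05 yaw relative to pelvis, measured from ref pose, into spine_rot, which blends two additive corrections on the front-left jacket bone.
# Left bracelets are offset, then AnimDynamics simulates bracelets and sleeves, a RigidBody node runs the jacket physics asset, and spring-driven AnimDynamics chains blend by dyn_alpha.
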AnimBlueprint /BRCosmetics/Characters/Player/Female/Medium/Bodies/F_MED_Devil_Rock/Meshes/F_MED_Devil_Rock_AnimBP
target_skeleton=&/F_MED_Devil_Rock_Skeleton
graphs [
 AnimGraph {nodes [
   n1 = Get spine_rot() -> (spine_rot:double)
   n2 = "Transform (Modify) Bone - Bone: dyn_jacket_fr_l_02"(
    ComponentPose=Local To Component(LocalPose=Copy Pose From Mesh() {bUseAttachedParent=true})
    Alpha=@n1
    Translation="0,0,0"
    Rotation="0,35,0"
    Scale="1,1,1"
   ) {
    rotationMode=BMM_Additive
    rotationSpace=BCS_BoneSpace
    translationMode=BMM_Additive
    translationSpace=BCS_BoneSpace
    alphaBoolBlend {blendOption=Linear}
    alphaScaleBiasClamp {bInterpResult=true, interpSpeedDecreasing=20, interpSpeedIncreasing=20}
    boneToModify {boneName=dyn_jacket_fr_l_02}
    rotation {yaw=35}
   }
   n3 = "Transform (Modify) Bone - Bone: dyn_jacket_fr_l_02"(ComponentPose=@n2, Alpha=@n1, Translation="0,3,3", Rotation="0,75,0", Scale="1,1,1") {
    rotationMode=BMM_Additive
    rotationSpace=BCS_BoneSpace
    translationMode=BMM_Additive
    translationSpace=BCS_BoneSpace
    alphaBoolBlend {blendOption=Linear}
    alphaScaleBiasClamp {bInterpResult=true, interpSpeedDecreasing=20, interpSpeedIncreasing=20}
    boneToModify {boneName=dyn_jacket_fr_l_02}
    rotation {yaw=75}
    translation {y=3, z=3}
   }
   n4 = "Transform (Modify) Bone - Bone: braceletB_l"(ComponentPose=@n3, Alpha=1, Translation="-0.867,0.133,0.391", Rotation="10.034,4.729,0.828", Scale="1,1,1") {
    rotationMode=BMM_Additive
    rotationSpace=BCS_BoneSpace
    translationMode=BMM_Additive
    translationSpace=BCS_BoneSpace
    alphaBoolBlend {blendOption=Linear}
    alphaScaleBiasClamp {bInterpResult=true, interpSpeedDecreasing=20, interpSpeedIncreasing=20}
    boneToModify {boneName=braceletB_l}
    rotation=<rot 10.034 4.729 0.828>
    translation=<-0.867 0.133 0.391>
   }
   n5 = "Transform (Modify) Bone - Bone: braceletA_l"(ComponentPose=@n4, Alpha=1, Translation="-1.186,0.395,0.229", Rotation="0,0,0", Scale="1,1,1") {
    rotationSpace=BCS_BoneSpace
    translationMode=BMM_Additive
    translationSpace=BCS_BoneSpace
    alphaBoolBlend {blendOption=Linear}
    alphaScaleBiasClamp {bInterpResult=true, interpSpeedDecreasing=20, interpSpeedIncreasing=20}
    boneToModify {boneName=braceletA_l}
    translation=<-1.186 0.3954 0.229>
   }
   n6 = Anim Dynamics - Bone: braceletA_l(ComponentPose=@n5, Alpha=0.9, ExternalForce="0,0,0") {
    angularDampingOverride=1
    bGravityOverrideInSimSpace=true
    bOverrideAngularDamping=true
    gravityScale=0.3
    simulationSpace=RootRelative
    alphaBoolBlend {blendOption=Linear}
    boundBone {boneName=braceletA_l}
    gravityOverride {z=-980}
    physicsBodyDefinitions [{
      sphereCollisionRadius=0
      boundBone {boneName=braceletA_l}
      boxExtents=<5 5 5>
      constraintSetup {angularLimitsMax=<2 2 2>, angularLimitsMin=<-2 -2 -2>, angularTarget=<0 0 0>, linearAxesMax=<0 0.2 0.2>, linearAxesMin=<-2 -0.2 -0.2>}
      localJointOffset=<-0.15 0 -0.15>
     }]
    relativeSpaceBone {boneName=lowerarm_l}
   }
   n7 = Anim Dynamics - Bone: braceletB_l(ComponentPose=@n6, Alpha=0.85, ExternalForce="0,0,0") {
    angularDampingOverride=1
    bGravityOverrideInSimSpace=true
    bOverrideAngularDamping=true
    gravityScale=0.3
    simulationSpace=RootRelative
    alphaBoolBlend {blendOption=Linear}
    boundBone {boneName=braceletB_l}
    gravityOverride {z=-980}
    physicsBodyDefinitions [{
      sphereCollisionRadius=0
      boundBone {boneName=braceletB_l}
      boxExtents=<5 5 5>
      constraintSetup {angularLimitsMax=<0 0 0>, angularLimitsMin=<0 0 0>, angularTarget=<0 0 0>, linearAxesMin=<-1.5 0 0>}
      localJointOffset=<-0.15 0 -0.15>
     }]
    relativeSpaceBone {boneName=lowerarm_l}
   }
   n8 = Anim Dynamics - Bone: driver_sleeve_outer_l(ComponentPose=@n7, Alpha=0.7, ExternalForce="0,0,0") {
    angularDampingOverride=1
    bGravityOverrideInSimSpace=true
    bOverrideAngularDamping=true
    gravityScale=0.31
    simulationSpace=RootRelative
    alphaBoolBlend {blendOption=Linear}
    boundBone {boneName=driver_sleeve_outer_l}
    componentLinearAccScale=<1 1 1>
    componentLinearVelScale=<1 1 1>
    gravityOverride {z=-980}
    physicsBodyDefinitions [{
      sphereCollisionRadius=0
      boundBone {boneName=driver_sleeve_outer_l}
      boxExtents=<5 5 5>
      constraintSetup {twistAxis=AxisZ, angularLimitsMax=<30 35 35>, angularLimitsMin=<-30 -35 -35>, angularTarget=<0 0 0>}
      localJointOffset=<0 -1 -2>
     }]
    relativeSpaceBone {boneName=lowerarm_l}
   }
   n9 = Anim Dynamics - Bone: driver_sleeve_inner_l(ComponentPose=@n8, Alpha=0.7, ExternalForce="0,0,0") {
    angularDampingOverride=1
    bGravityOverrideInSimSpace=true
    bOverrideAngularDamping=true
    gravityScale=0.31
    simulationSpace=RootRelative
    alphaBoolBlend {blendOption=Linear}
    boundBone {boneName=driver_sleeve_inner_l}
    componentLinearAccScale=<1 1 1>
    componentLinearVelScale=<1 1 1>
    gravityOverride {z=-980}
    physicsBodyDefinitions [{
      sphereCollisionRadius=0
      boundBone {boneName=driver_sleeve_inner_l}
      boxExtents=<5 5 5>
      constraintSetup {twistAxis=AxisZ, angularLimitsMax=<30 35 35>, angularLimitsMin=<-30 -35 -35>, angularTarget=<0 0 0>}
      localJointOffset=<0 -1 -2>
     }]
    relativeSpaceBone {boneName=lowerarm_l}
   }
   n10 = Anim Dynamics - Bone: driver_sleeve_outer_r(ComponentPose=@n9, Alpha=0.7, ExternalForce="0,0,0") {
    angularDampingOverride=1
    bGravityOverrideInSimSpace=true
    bOverrideAngularDamping=true
    gravityScale=0.31
    simulationSpace=RootRelative
    alphaBoolBlend {blendOption=Linear}
    boundBone {boneName=driver_sleeve_outer_r}
    componentLinearAccScale=<1 1 1>
    componentLinearVelScale=<1 1 1>
    gravityOverride {z=-980}
    physicsBodyDefinitions [{
      sphereCollisionRadius=0
      boundBone {boneName=driver_sleeve_outer_r}
      boxExtents=<5 5 5>
      constraintSetup {twistAxis=AxisZ, angularLimitsMax=<30 33 35>, angularLimitsMin=<-30 -33 -35>, angularTarget=<0 0 0>}
      localJointOffset=<0 1 2>
     }]
    relativeSpaceBone {boneName=lowerarm_l}
   }
   n11 = Anim Dynamics - Bone: driver_sleeve_inner_r(ComponentPose=@n10, Alpha=0.7, ExternalForce="0,0,0") {
    angularDampingOverride=1
    bGravityOverrideInSimSpace=true
    bOverrideAngularDamping=true
    gravityScale=0.31
    simulationSpace=RootRelative
    alphaBoolBlend {blendOption=Linear}
    boundBone {boneName=driver_sleeve_inner_r}
    componentLinearAccScale=<1 1 1>
    componentLinearVelScale=<1 1 1>
    gravityOverride {z=-980}
    physicsBodyDefinitions [{
      sphereCollisionRadius=0
      boundBone {boneName=driver_sleeve_inner_r}
      boxExtents=<5 5 5>
      constraintSetup {twistAxis=AxisZ, angularLimitsMax=<30 33 35>, angularLimitsMin=<-30 -33 -35>, angularTarget=<0 0 0>}
      localJointOffset=<0 1 2>
     }]
    relativeSpaceBone {boneName=lowerarm_l}
   }
   n12 = RigidBody(ComponentPose=@n11, Alpha=0.65) {
    bClampLinearTranslationLimitToRefPose=true
    bOverrideWorldGravity=true
    overridePhysicsAsset=&/F_MED_Devil_Rock_Dyn_Jacket
    simulationSpace=BaseBoneSpace
    alphaBoolBlend {blendOption=Linear}
    baseBoneRef {boneName=spine_01}
    componentAppliedLinearAccClamp=<3000 3000 3000>
    componentLinearAccScale=<0.35 0.2 0.2>
    componentLinearVelScale=<0.35 0.2 0.2>
    overrideWorldGravity {z=-980}
   }
   n13 = Get dyn_alpha() -> (dyn_alpha:double)
   n14 = Anim Dynamics - Bone: chain_sm_01_r(ComponentPose=@n12, Alpha=@n13, ExternalForce="0,0,0") {
    angularDampingOverride=1
    angularSpringConstant=2500
    bAngularSpring=true
    bGravityOverrideInSimSpace=true
    bOverrideAngularDamping=true
    bUseGravityOverride=true
    gravityScale=0.95
    numSolverIterationsPostUpdate=4
    numSolverIterationsPreUpdate=16
    simulationSpace=RootRelative
    alphaBoolBlend {blendOption=Linear}
    boundBone {boneName=chain_sm_01_r}
    componentAppliedLinearAccClamp=<3000 3000 1700>
    componentLinearAccScale=<8 8 8>
    componentLinearVelScale=<5 5 5>
    gravityOverride {z=-980}
    physicsBodyDefinitions [{
      sphereCollisionRadius=0
      boundBone {boneName=chain_sm_01_r}
      boxExtents=<15 8 2>
      constraintSetup {twistAxis=AxisZ, angularLimitsMax=<3 0.5 0.25>, angularLimitsMin=<-3 -60 -0.25>, angularTarget=<0 0 0>}
      localJointOffset=<-5 0 5>
     }]
   }
   n15 = Anim Dynamics - Bone: chain_sm_02_r(ComponentPose=@n14, Alpha=@n13, ExternalForce="0,0,0") {
    angularDampingOverride=1
    angularSpringConstant=2500
    bAngularSpring=true
    bGravityOverrideInSimSpace=true
    bOverrideAngularDamping=true
    bUseGravityOverride=true
    gravityScale=0.95
    numSolverIterationsPostUpdate=4
    numSolverIterationsPreUpdate=16
    simulationSpace=RootRelative
    alphaBoolBlend {blendOption=Linear}
    boundBone {boneName=chain_sm_02_r}
    componentAppliedLinearAccClamp=<3000 3000 1700>
    componentLinearAccScale=<8 8 8>
    componentLinearVelScale=<5 5 5>
    gravityOverride {z=-980}
    physicsBodyDefinitions [{
      sphereCollisionRadius=0
      boundBone {boneName=chain_sm_02_r}
      boxExtents=<11 8 2>
      constraintSetup {twistAxis=AxisZ, angularLimitsMax=<3 0.5 0.25>, angularLimitsMin=<-3 -60 -0.25>, angularTarget=<0 0 0>}
      localJointOffset=<-5 0 5>
     }]
   }
   n16 = Anim Dynamics - Bone: chain_lrg_01_r(ComponentPose=@n15, Alpha=@n13, ExternalForce="0,0,0") {
    angularDampingOverride=1
    angularSpringConstant=2500
    bAngularSpring=true
    bGravityOverrideInSimSpace=true
    bOverrideAngularDamping=true
    bUseGravityOverride=true
    gravityScale=0.95
    numSolverIterationsPostUpdate=4
    numSolverIterationsPreUpdate=16
    simulationSpace=RootRelative
    alphaBoolBlend {blendOption=Linear}
    boundBone {boneName=chain_lrg_01_r}
    componentAppliedLinearAccClamp=<3000 3000 1700>
    componentLinearAccScale=<8 8 8>
    componentLinearVelScale=<5 5 5>
    gravityOverride {z=-980}
    physicsBodyDefinitions [{
      sphereCollisionRadius=0
      boundBone {boneName=chain_lrg_01_r}
      boxExtents=<20 11 2>
      constraintSetup {twistAxis=AxisZ, angularLimitsMax=<3 0.5 0.25>, angularLimitsMin=<-3 -60 -0.25>, angularTarget=<0 0 0>}
      localJointOffset=<-5 0 5>
     }]
   }
   n17 = Anim Dynamics - Bone: chain_lrg_02_r(ComponentPose=@n16, Alpha=@n13, ExternalForce="0,0,0") {
    angularDampingOverride=1
    angularSpringConstant=2500
    bAngularSpring=true
    bGravityOverrideInSimSpace=true
    bOverrideAngularDamping=true
    bUseGravityOverride=true
    gravityScale=0.95
    numSolverIterationsPostUpdate=4
    numSolverIterationsPreUpdate=16
    simulationSpace=RootRelative
    alphaBoolBlend {blendOption=Linear}
    boundBone {boneName=chain_lrg_02_r}
    componentAppliedLinearAccClamp=<3000 3000 1700>
    componentLinearAccScale=<8 8 8>
    componentLinearVelScale=<5 5 5>
    gravityOverride {z=-980}
    physicsBodyDefinitions [{
      sphereCollisionRadius=0
      boundBone {boneName=chain_lrg_02_r}
      boxExtents=<17 13 2>
      constraintSetup {twistAxis=AxisZ, angularLimitsMax=<3 0.5 0.25>, angularLimitsMin=<-3 -60 -0.25>, angularTarget=<0 0 0>}
      localJointOffset=<-5 0 5>
     }]
   }
   n18 = Anim Dynamics - Bone: chain_sm_01_l(ComponentPose=@n17, Alpha=@n13, ExternalForce="0,0,0") {
    angularDampingOverride=1
    angularSpringConstant=2000
    bAngularSpring=true
    bGravityOverrideInSimSpace=true
    bOverrideAngularDamping=true
    gravityScale=0.95
    numSolverIterationsPostUpdate=4
    numSolverIterationsPreUpdate=16
    simulationSpace=RootRelative
    alphaBoolBlend {blendOption=Linear}
    boundBone {boneName=chain_sm_01_l}
    componentAppliedLinearAccClamp=<3000 30000 7000>
    componentLinearAccScale=<1 1 1>
    componentLinearVelScale=<1 1 1>
    physicsBodyDefinitions [{
      sphereCollisionRadius=0
      boundBone {boneName=chain_sm_01_l}
      boxExtents=<15 8 2>
      constraintSetup {twistAxis=AxisZ, angularLimitsMax=<3 0.5 0.25>, angularLimitsMin=<-3 -60 -0.25>, angularTarget=<0 0 0>}
      localJointOffset=<5 0 -5>
     }]
   }
   n19 = Anim Dynamics - Bone: chain_sm_02_l(ComponentPose=@n18, Alpha=@n13, ExternalForce="0,0,0") {
    angularDampingOverride=1
    angularSpringConstant=1000
    bAngularSpring=true
    bGravityOverrideInSimSpace=true
    bOverrideAngularDamping=true
    gravityScale=0.95
    numSolverIterationsPostUpdate=4
    numSolverIterationsPreUpdate=16
    simulationSpace=RootRelative
    alphaBoolBlend {blendOption=Linear}
    boundBone {boneName=chain_sm_02_l}
    componentAppliedLinearAccClamp=<3000 30000 7000>
    componentLinearAccScale=<1 1 1>
    componentLinearVelScale=<1 1 1>
    physicsBodyDefinitions [{
      sphereCollisionRadius=0
      boundBone {boneName=chain_sm_02_l}
      boxExtents=<11 8 2>
      constraintSetup {twistAxis=AxisZ, angularLimitsMax=<3 0.5 0.25>, angularLimitsMin=<-3 -60 -0.25>, angularTarget=<0 0 0>}
      localJointOffset=<5 0 -5>
     }]
   }
   n20 = Anim Dynamics - Bone: chain_lrg_01_l(ComponentPose=@n19, Alpha=@n13, ExternalForce="0,0,0") {
    angularDampingOverride=1
    angularSpringConstant=2000
    bAngularSpring=true
    bGravityOverrideInSimSpace=true
    bOverrideAngularDamping=true
    gravityScale=0.95
    numSolverIterationsPostUpdate=4
    numSolverIterationsPreUpdate=16
    simulationSpace=RootRelative
    alphaBoolBlend {blendOption=Linear}
    boundBone {boneName=chain_lrg_01_l}
    componentAppliedLinearAccClamp=<3000 30000 7000>
    componentLinearAccScale=<0.5 0.5 0.5>
    componentLinearVelScale=<0.5 0.5 0.5>
    physicsBodyDefinitions [{
      sphereCollisionRadius=0
      boundBone {boneName=chain_lrg_01_l}
      boxExtents=<20 11 2>
      constraintSetup {twistAxis=AxisZ, angularLimitsMax=<3 0.5 0.25>, angularLimitsMin=<-3 -60 -0.25>, angularTarget=<0 0 0>}
      localJointOffset=<5 0 -5>
     }]
   }
   n21 = Anim Dynamics - Bone: chain_lrg_02_l(ComponentPose=@n20, Alpha=@n13, ExternalForce="0,0,0") {
    angularDampingOverride=1
    angularSpringConstant=1000
    bAngularSpring=true
    bGravityOverrideInSimSpace=true
    bOverrideAngularDamping=true
    gravityScale=0.95
    numSolverIterationsPostUpdate=4
    numSolverIterationsPreUpdate=16
    simulationSpace=RootRelative
    alphaBoolBlend {blendOption=Linear}
    boundBone {boneName=chain_lrg_02_l}
    componentAppliedLinearAccClamp=<3000 30000 7000>
    componentLinearAccScale=<1 1 1>
    componentLinearVelScale=<1 1 1>
    physicsBodyDefinitions [{
      sphereCollisionRadius=0
      boundBone {boneName=chain_lrg_02_l}
      boxExtents=<17 13 2>
      constraintSetup {twistAxis=AxisZ, angularLimitsMax=<3 0.5 0.25>, angularLimitsMin=<-3 -60 -0.25>, angularTarget=<0 0 0>}
      localJointOffset=<5 0 -5>
     }]
   }
   Output Pose(Result=Component To Local(ComponentPose=@n21)) {layerGroup=None}
  ]}
 EventGraph {nodes [
   Event BlueprintUpdateAnimation()
   n1 = BreakTransform(InTransform=GetDeltaTransformFromRefPose(self=GetOwnerSkeletalMeshComponent(), BoneName=spine_05, BaseName=pelvis))
   > Set spine_rot(spine_rot:double=MapRangeClamped(Value=BreakRotator(InRot=@n1.Rotation).Yaw, InRangeA=15, InRangeB=35)) -> (Output_Get:double)
  ]}
]
variables [
 dyn_alpha {type=number, default=0.7}
 spine_rot {type=number}
]
