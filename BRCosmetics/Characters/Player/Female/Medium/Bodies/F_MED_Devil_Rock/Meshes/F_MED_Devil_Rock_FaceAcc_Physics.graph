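# Physics asset for the F_MED_Devil_Rock face accessory mesh, giving its left and right hair bones collision bodies and joint limits.
# Box bodies form mirrored three-bone hairA and two-bone hairB chains, while single hairD bones use small capsules.
# Each chain root and hairD bone constrains to the head, which has two overlapping capsules.
# All joints share the same swing limits, with tighter twist on hairD.
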PhysicsAsset /BRCosmetics/Characters/Player/Female/Medium/Bodies/F_MED_Devil_Rock/Meshes/F_MED_Devil_Rock_FaceAcc_Physics
bodies {
 hairA_01_l [{shapeType=Box, extentX=4.0659, extentY=1.6434, extentZ=4.5, center=<1.7373 0 0.0001>, rotation=<rot 0 0 90>}]
 hairA_01_r [{shapeType=Box, extentX=4.157, extentY=1.5, extentZ=4, center=<-1.8052 0 0>, rotation=<rot 0 0 90>}]
 hairA_02_l [{shapeType=Box, extentX=3.8363, extentY=1.5824, extentZ=4.6743, center=<1.6845 0 0>, rotation=<rot 0 0 90>}]
 hairA_02_r [{shapeType=Box, extentX=4.0856, extentY=1.5, extentZ=4, center=<-1.5857 0 0.0001>, rotation=<rot 0 0 90>}]
 hairA_03_l [{shapeType=Box, extentX=4.2733, extentY=1.8416, extentZ=2.8186, center=<1.9746 0 0.0001>, rotation=<rot 0 0 90>}]
 hairA_03_r [{shapeType=Box, extentX=4.8843, extentY=1.4605, extentZ=3.237, center=<-2.0618 0 -0.0001>, rotation=<rot 0 0 90>}]
 hairB_01_l [{shapeType=Box, extentX=3.4515, extentY=1.7556, extentZ=3.4803, center=<1.1998 0 0.0001>, rotation=<rot 0 0 90>}]
 hairB_01_r [{shapeType=Box, extentX=4.1282, extentY=1.4194, extentZ=5.1294, center=<-1.3282 0 0>, rotation=<rot 0 0 90>}]
 hairB_02_l [{shapeType=Box, extentX=3.1589, extentY=0.9317, extentZ=2.804, center=<1.3464 0 0.0001>, rotation=<rot 0 0 90>}]
 hairB_02_r [{shapeType=Box, extentX=3.7072, extentY=0.8765, extentZ=3.5781, center=<-1.0275 0 0>, rotation=<rot 0 0 90>}]
 hairD_01_l [{shapeType=Capsule, length=1.5, radius=0.9411, center=<1.2467 0 0>, rotation=<rot -90 -0.0149 0.0149>}]
 hairD_01_r [{shapeType=Capsule, length=1.5, radius=1.005, center=<-1.1205 0 0.0001>, rotation=<rot 90 1.2659 1.266>}]
 head [
  {shapeType=Capsule, length=14.5437, radius=6.7123, rotation=<rot -90 -165.9637 -164.0347>}
  {shapeType=Capsule, length=9.7937, radius=5.4623, center=<-5.256 2.3106 0.0001>, rotation=<rot -90 0 60.0033>}
 ]
}
constraints [
 {bone1Name=hairA_01_l, bone2Name=head, swing1LimitDegrees=45, swing2LimitDegrees=45, twistLimitDegrees=10}
 {bone1Name=hairA_02_l, bone2Name=hairA_01_l, swing1LimitDegrees=45, swing2LimitDegrees=45, twistLimitDegrees=10}
 {bone1Name=hairA_03_l, bone2Name=hairA_02_l, swing1LimitDegrees=45, swing2LimitDegrees=45, twistLimitDegrees=10}
 {bone1Name=hairB_01_l, bone2Name=head, swing1LimitDegrees=45, swing2LimitDegrees=45, twistLimitDegrees=10}
 {bone1Name=hairB_02_l, bone2Name=hairB_01_l, swing1LimitDegrees=45, swing2LimitDegrees=45, twistLimitDegrees=10}
 {bone1Name=hairA_01_r, bone2Name=head, swing1LimitDegrees=45, swing2LimitDegrees=45, twistLimitDegrees=10}
 {bone1Name=hairA_02_r, bone2Name=hairA_01_r, swing1LimitDegrees=45, swing2LimitDegrees=45, twistLimitDegrees=10}
 {bone1Name=hairA_03_r, bone2Name=hairA_02_r, swing1LimitDegrees=45, swing2LimitDegrees=45, twistLimitDegrees=10}
 {bone1Name=hairB_01_r, bone2Name=head, swing1LimitDegrees=45, swing2LimitDegrees=45, twistLimitDegrees=10}
 {bone1Name=hairB_02_r, bone2Name=hairB_01_r, swing1LimitDegrees=45, swing2LimitDegrees=45, twistLimitDegrees=10}
 {bone1Name=hairD_01_r, bone2Name=head, swing1LimitDegrees=45, swing2LimitDegrees=45, twistLimitDegrees=5}
 {bone1Name=hairD_01_l, bone2Name=head, swing1LimitDegrees=45, swing2LimitDegrees=45, twistLimitDegrees=5}
]
